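# Default Lit material for stone chip effects, flagged for use on Cascade and Niagara mesh particles.
# A Color vector parameter, defaulting to mid grey, feeds the diffuse input of a ConvertFromDiffSpec function.
# A wood chips texture sample sits unconnected, and no links from either the function or the texture to the material outputs appear in the graph.
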Material /Game/Effects/Fort_Effects/Materials/Stone/M_Stone_Chip
graph {
 properties {bUsedWithMeshParticles=true, bUsedWithNiagaraMeshParticles=true, pixelDepthOffsetMode=PDOM_Legacy, usedShadingModels=Default Lit}
 nodes [
  MaterialFunctionCall(DiffuseColor=VectorParameter() {parameterName=Color, defaultValue=<rgba 0.5 0.5 0.5 1>}.RGB) {materialFunction=&ConvertFromDiffSpec}
  TextureSample() {texture=/Game/Effects/Fort_Effects/Textures/Wood/T_Wood_Chips}
 ]
}
